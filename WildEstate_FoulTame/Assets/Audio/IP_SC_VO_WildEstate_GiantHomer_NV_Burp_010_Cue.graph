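# Sound cue for a non-verbal burp voice-over from the Giant Homer character in the WildEstate FoulTame content.
# A single wave player feeds the Burp_010 recording straight to the output, with no randomization or modulation.
# It uses the narrator VO sound class and borrows a lighthouse horn ambient attenuation with a very large max distance, so the burp carries far across the map.
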SoundCue /WildEstate_FoulTame/Assets/Audio/IP_SC_VO_WildEstate_GiantHomer_NV_Burp_010_Cue
root=WavePlayer_0
nodes [{id=WavePlayer_0, properties {soundWaveAssetPtr=&/vo_IP_WildEstate_GiantHomer_NV_Burp_010}}]
properties {
 attenuationSettings=/Game/Sounds/Attenuation/Fort_Tile_Ambient_LighthouseHornApollo_Att
 maxDistance=65000
 soundClassObject=/Game/Sounds/FortSoundClassesAndModes/Fort_VO_Narrator
 volumeMultiplier=1
}
