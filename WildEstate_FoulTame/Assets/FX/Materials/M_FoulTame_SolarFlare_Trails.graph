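# Translucent VFX material for FoulTame solar flare trails on ribbons, beams, sprites and mesh particles.
# Dynamic-parameter offset and scale drive UV1, where panning noise distorts a scrolling trail mask combined with smoke streaks and magnetic waves inside a square edge mask.
# That intensity blends two colors, plus wave RGB and optional particle-color tint, for emissive.
# Opacity adds particle alpha, depth fade and near-camera fade, skipped on low quality and mobile.
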Material /WildEstate_FoulTame/Assets/FX/Materials/M_FoulTame_SolarFlare_Trails
graph {
 properties {
  bUsedWithBeamTrails=true
  bUsedWithNiagaraMeshParticles=true
  bUsedWithNiagaraRibbons=true
  bUsedWithNiagaraSprites=true
  bUsedWithParticleSprites=true
  blendMode=BLEND_Translucent
  pixelDepthOffsetMode=PDOM_Legacy
  usedShadingModels=Default Lit
 }
 nodes [
  n1 = ParticleColor()
  n2 = DynamicParameter() {paramNames [RibbonOffset,--None--,UV1_XScale,UV1_YScale]}
  n3 = Add(
   A=AppendVector(A=@n2.RibbonOffset, B=Constant())
   B=Multiply(A=AppendVector(A=@n2.UV1_XScale, B=@n2.UV1_XScale), B=TextureCoordinate() {coordinateIndex=1})
  )
  n4 = Multiply(A=@n3, B=AppendVector(A=ScalarParameter() {defaultValue=-1, parameterName=mainTile}, B=Constant() {r=1}))
  n5 = VectorParameter() {parameterName=DistTileSpeed, defaultValue=<rgba 1 2 -2 0>}
  n6 = Panner(Coordinate=Multiply(A=@n3, B=ComponentMask(None=@n5.RGB) {g=true, r=true}), Time=Time(), Speed=AppendVector(A=@n5.B, B=@n5.B)) {bFractionalPart=true, speedX=-0.6}
  n7 = Multiply(
   A=LinearInterpolate(Alpha=TextureSample(UVs=@n6) {samplerType=SAMPLERTYPE_Grayscale, texture=&LowResBlurredNoise}.G) {constA=-1}
   B=ScalarParameter() {defaultValue=0.1, parameterName=DistAmnt}
  )
  n8 = ScalarParameter() {defaultValue=1, parameterName=mainSpeed}
  n9 = Constant()
  n10 = TextureSampleParameter2D(UVs=Panner(Coordinate=Add(A=@n4, B=@n7), Speed=AppendVector(A=@n8, B=@n9)) {bFractionalPart=true}) {parameterName=AlphaTexture, texture=/Game/Effects/Fort_Effects/Textures/Smoke/T_Lobber_Trail_Mask}
  n11 = Multiply(A=ScalarParameter() {defaultValue=0.6, parameterName=ExtraAdd}, B=@n10.G)
  n12 = VectorParameter() {parameterName=ChromaTileSpeed, defaultValue=<rgba 0.1 0.2 -0.1 0.05>}
  n13 = ComponentMask(None=@n12.RGB) {g=true, r=true}
  n14 = TextureCoordinate() {coordinateIndex=1}
  n15 = TextureSample(UVs=Panner(Coordinate=Multiply(A=@n13, B=@n14) {constA=0.3}, Speed=AppendVector(A=@n12.B, B=@n12.B)) {speedX=-0.1, speedY=0.05}) {texture=/Game/Athena/Prototype/Blueprints/Sneezy/Textures/T_Sneezy_MagneticWaves}
  n16 = ScalarParameter() {defaultValue=1, parameterName=CoorAmnt}
  n17 = AppendVector(A=Constant() {r=0.5}, B=ScalarParameter() {defaultValue=4, parameterName=StreaksTiles})
  n18 = ScalarParameter() {defaultValue=-1, parameterName=StreaksSpeed}
  n19 = Constant()
  n20 = TextureSample(UVs=Panner(Coordinate=Multiply(A=@n17, B=@n3) {constA=5}, Speed=AppendVector(A=@n18, B=@n19)) {speedX=-1}) {texture=&T_Inky_Smoke_Tile}
  n21 = Multiply(A=@n10.G, B=Add(A=Multiply(A=@n15.G, B=@n16), B=Multiply(A=@n10.G, B=@n20.G)))
  n22 = ScalarParameter() {defaultValue=5, parameterName=EdgeMult}
  n23 = ScalarParameter() {defaultValue=3, parameterName=EdgeExp}
  n24 = Multiply(
   A=ScalarParameter() {defaultValue=1, parameterName=bright}
   B=Multiply(
    A=Add(A=@n11, B=@n21)
    B=MaterialFunctionCall!(UVs, Mult=@n22, Exp=@n23) {materialFunction=/Game/Effects/Fort_Effects/MaterialFunctions/MF_SquareEdgeMask}.SquareMask
   )
  )
  n25 = ScalarParameter() {defaultValue=90, parameterName=NearFade}
  n26 = Saturate(None=Power(
    Base=MaterialFunctionCall(Distance=@n25) {materialFunction=&Near_Camera_Fade}.Result
    Exp=ScalarParameter() {defaultValue=5, parameterName=NearFadeExp}
   ))
  n27 = Constant() {r=1}
  n28 = FeatureLevelSwitch(Default=@n26, ES2=@n27, ES3_1=@n27, SM4_REMOVED=@n26, SM5=@n26, SM6=@n26)
  DepthFade(
   Opacity=Multiply(A=Multiply(A=@n1.A, B=@n24), B=QualitySwitch(Default=@n28, Low=@n27, High=@n28, Medium=@n28))
   FadeDistance=DynamicParameter() {parameterIndex=1, defaultValue=<rgba 0 1 1 1>, paramNames [FadeDistance,Param2,Param3,Param4]}.FadeDistance
  )
  n29 = Add(
   A=@n15.RGB
   B=LinearInterpolate(
    A=VectorParameter() {parameterName=Color2, defaultValue=<rgba 0.0002 0.1793 1 0>}.RGB
    B=VectorParameter() {parameterName=Color, defaultValue=<rgba 0.5729 0.8385 5 0>}.RGB
    Alpha=Saturate(None=@n24)
   )
  )
  MaterialFunctionCall(Emissive=LinearInterpolate(A=@n29, B=Multiply(A=@n29, B=@n1.RGB), Alpha=ScalarParameter() {parameterName=Use_ParticleColor})) {materialFunction=&MF_FX_EmissiveAdjustments}
  ScalarParameter() {parameterName=FadeDistance}
 ]
}
