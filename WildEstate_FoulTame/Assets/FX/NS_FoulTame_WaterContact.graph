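# Niagara water-contact effect for WildEstate FoulTame, in which bubbles rise to a surface and pop.
# Bubbles spawns mesh particles on a looping emitter and moves them along a curve-driven path from below, scaled by User.BubbleScale.
# Age curves on each particle stagger splash, wave and two ripple attributes for its four sprite renderers, with ripple tint varying by time of day.
# The Foam sprite emitter and the Bubbles region mask are disabled.
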NiagaraSystem /WildEstate_FoulTame/Assets/FX/NS_FoulTame_WaterContact
dependencies {
 usedDataInterfaces [&NiagaraDataInterfaceCurve, &NiagaraDataInterfaceColorCurve]
 usedDynamicInputs []
 usedModules [&SystemState]
}
emitters {
 Bubbles {
  emitterUpdateScript [
   &EmitterState(Inactive Response=::NewEnumerator1, Life Cycle Mode=::NewEnumerator1, Loop Delay=2, Loop Duration=2, UseLoopDelay=true)
   &SpawnRate(SpawnRate=2, Use Spawn Probability=false)
   /SetVariables(Emitter.BubbleEmissive:NiagaraFloat=&FNTimeOfDayScalar(Night Scale=0.5, Morning Scale=0.4, Day Scale=0.3, Evening Scale=0.4)) {isSetParameters=true}
  ]
  particleSpawnScript [
   &InitializeParticle(
    Color=<rgba 0.8099 0.8654 1 0.4>
    Lifetime Max=4
    Lifetime Min=3.5
    Lifetime Mode=::NewEnumerator1
    Mesh Scale Mode=::NewEnumerator3
    Mesh Uniform Scale=&Multiply_Float(A=$User.BubbleScale, B=&RandomRangeFloat(Minimum=0.04, Maximum=0.1))
    Position Offset=<0 -2500 0>
    Position Offset Coordinate Space=::Local
    Sprite Rotation Mode=::NewEnumerator2
    UsePositionOffset=true
    Write Material Random=true
   )
   &ShapeLocation(
    Disc Coverage=1
    MaskSpawnGroup=false
    Offset Mode=::NewEnumerator0
    Override Seed=false
    Ring Radius=&Multiply_Float(A=$User.BubbleScale, B=20)
    Shape Primitive=::NewEnumerator4
    UseOwnerScale=true
   )
   /SetVariables(
    Particles.SplashIndex:NiagaraFloat=&RandomRangeFloat(Minimum=0, Maximum=4)
    Particles.SplashVelocity:Vector3f=<0 0 1>
   ) {isSetParameters=true}
  ]
  particleUpdateScript [
   &ParticleState()
   &ScaleMeshSize(Scale Factor=&VectorFromFloat(Value=&FloatFromCurve(FloatCurve=DataInterface() {propertyValues {Curve {keys [<at 0 0.5 RCIM_Cubic>,<at 0.7 1>,<at 1 0>]}}})))
   &DynamicMaterialParameters(
    Index 0 Param 1=&FloatFromCurve(FloatCurve=DataInterface() {propertyValues {Curve {keys [
         <at 0 0>
         {interpMode=RCIM_Cubic, leaveTangent=1.2072, tangentMode=RCTM_Break, time=0.5, value=0}
         {arriveTangent=3.283, tangentMode=RCTM_User, time=0.8, value=1}
        ]}}})
    Index 0 Param 4=1.1
    Module.Index0Param0Write=true
    Module.Index0Param1Write=true
    Module.Index0Param2Write=true
    Module.Index0Param3Write=true
   )
   &SolveForcesAndVelocity(Clamp Velocity=false, Limit Acceleration=false)
   &RegionMask(
    Inner Float=0
    Offset:Vector3f=<0 -30 94>
    Outer Float=1
    Region Falloff=0
    Region Radius:NiagaraFloat=&Multiply_Float(A=$User.BubbleScale, B=120)
    UseOffset=true
    UseRotation=false
   ) {enabled=false}
   /SetVariables(
    Particles.SplashDynamicParamenter:Vector4f=&MakeVector4(
     X=&FloatFromCurve(FloatCurve=DataInterface() {propertyValues {Curve {keys [
          <at 0 1>
          {interpMode=RCIM_Cubic, leaveTangent=127.6574, tangentMode=RCTM_Break, time=0.8, value=1}
          {arriveTangent=27.4703, tangentMode=RCTM_User, time=1, value=17}
         ]}}})
     Y=0.7
     Z=0
     W=0
    )
    Particles.SplashPosition:NiagaraPosition=&Add_Vector(
     A=Linked() {linkedVariable {name=Particles.Position, type=NiagaraPosition}}
     B=&MakeVector(
      X=0
      Y=0
      Z=&FloatFromCurve(
       FloatCurve=DataInterface() {propertyValues {Curve {keys [
           {interpMode=RCIM_Cubic, leaveTangent=156.6917, tangentMode=RCTM_Break, time=0.55, value=-10}
           {interpMode=RCIM_Cubic, tangentMode=RCTM_Break, time=0.75, value=5}
           {arriveTangent=-261.0323, interpMode=RCIM_Cubic, tangentMode=RCTM_Break, time=1, value=-50}
          ]}}}
       Scale Curve=&Multiply_Float(A=$User.BubbleScale, B=&RandomRangeFloat(Minimum=0.5, Maximum=1))
      )
     )
    )
    Particles.SplashSize:Vector2f=&Multiply_Vector2D(
     A=&Multiply_Vector2D(
      A=<0.8 0.8>
      B=&Vector2DFromFloat(Value=&Lerp_Float(A=0, B=1, Alpha=$StackContext.RegionMask.Float))
     )
     B=&Vector2DFromFloat(Value=&FloatFromCurve(
       FloatCurve=DataInterface() {propertyValues {Curve {keys [<at 0.6 0>,<at 0.8 1>]}}}
       Scale Curve=&Multiply_Float(A=$User.BubbleScale, B=&RandomRangeFloat(Minimum=20, Maximum=25))
      ))
    )
    Particles.SplashVector:Vector3f=<0 0 1>
   ) {isSetParameters=true}
   /SetVariables(
    Particles.WaveColor:LinearColor=&MultiplyLinearColor(
     Linear Color=&ColorFromCurve(ColorCurve=DataInterface() {propertyValues {
        AlphaCurve {keys [<at 0.6 0>,<at 0.7 1>,<at 0.8 0 RCIM_Cubic>]}
        BlueCurve {keys [<at 0.8093 1 RCIM_Cubic>,<at 1 0.3542 RCIM_Cubic>]}
        GreenCurve {keys [<at 0.8093 1 RCIM_Cubic>,<at 1 0.3542 RCIM_Cubic>]}
        RedCurve {keys [<at 0.8093 1 RCIM_Cubic>,<at 1 0.3542 RCIM_Cubic>]}
       }})
     Scale Factor=&MakeVector4_FromLinearColor(LinearColor=$Particles.Initial.Color)
    )
    Particles.WaveSize:Vector2f=&Vector2DFromFloat(Value=&MakeFloatFromVector2d(Vector 2D=&Vector2DFromFloat(Value=&FloatFromCurve(
        FloatCurve=DataInterface() {propertyValues {Curve {keys [<at 0 0>,<at 0.6 0>,<at 0.8 1>]}}}
        Scale Curve=&Multiply_Float(A=$User.BubbleScale, B=&RandomRangeFloat(Minimum=25, Maximum=30))
       ))))
   ) {isSetParameters=true}
   /SetVariables(
    Particles.RippleColor:LinearColor=&MultiplyLinearColor(
     Linear Color=&MultiplyLinearColor(
      Linear Color=&ColorFromCurve(ColorCurve=DataInterface() {propertyValues {
         AlphaCurve {keys [<at 0.6765 0>,<at 0.7462 1>,<at 0.9056 0 RCIM_Cubic>]}
         BlueCurve {keys [<at 0.8093 1 RCIM_Cubic>,<at 1 0.3542 RCIM_Cubic>]}
         GreenCurve {keys [<at 0.8093 1 RCIM_Cubic>,<at 1 0.3542 RCIM_Cubic>]}
         RedCurve {keys [<at 0.8093 1 RCIM_Cubic>,<at 1 0.3542 RCIM_Cubic>]}
        }})
      Scale Factor=&MakeVector4_FromLinearColor(LinearColor=&FNTimeOfDayColor(Night Color=<rgba 0.556 0.5008 1 1>, Morning Color=<rgba 1 0.6695 0.6095 1>, Day Color=<rgba 1 1 1 1>, Evening Color=<rgba 1 1 1 1>))
     )
     Scale Factor=&MakeVector4_FromLinearColor(LinearColor=$Particles.Initial.Color)
    )
    Particles.RippleFacing:Vector3f=<0 0 1>
    Particles.RippleSize:Vector2f=&Vector2DFromFloat(Value=&FloatFromCurve(
      FloatCurve=DataInterface() {propertyValues {Curve {keys [<at 0.5 0>,<at 1 1>]}}}
      Scale Curve=&Multiply_Float(A=$User.BubbleScale, B=35)
     ))
   ) {isSetParameters=true}
   /SetVariables(
    Particles.Ripple2Color:LinearColor=&MultiplyLinearColor(
     Linear Color=&MultiplyLinearColor(
      Linear Color=&ColorFromCurve(ColorCurve=DataInterface() {propertyValues {
         AlphaCurve {keys [<at 0.8143 0>,<at 0.8876 1>,<at 0.9661 0 RCIM_Cubic>]}
         BlueCurve {keys [<at 0.8093 1 RCIM_Cubic>,<at 1 0.3542 RCIM_Cubic>]}
         GreenCurve {keys [<at 0.8093 1 RCIM_Cubic>,<at 1 0.3542 RCIM_Cubic>]}
         RedCurve {keys [<at 0.8093 1 RCIM_Cubic>,<at 1 0.3542 RCIM_Cubic>]}
        }})
      Scale Factor=&MakeVector4_FromLinearColor(LinearColor=&FNTimeOfDayColor(Night Color=<rgba 0.556 0.5008 1 1>, Morning Color=<rgba 1 0.6695 0.6095 1>, Day Color=<rgba 1 1 1 1>, Evening Color=<rgba 1 1 1 1>))
     )
     Scale Factor=&MakeVector4_FromLinearColor(LinearColor=$Particles.Initial.Color)
    )
    Particles.Ripple2Size:Vector2f=&Vector2DFromFloat(Value=&FloatFromCurve(
      FloatCurve=DataInterface() {propertyValues {Curve {keys [
          {interpMode=RCIM_Cubic, leaveTangent=0.6678, tangentMode=RCTM_Break, time=0.6, value=0}
          {arriveTangent=2.5069, tangentMode=RCTM_User, time=1, value=0.65}
         ]}}}
      Scale Curve=&Multiply_Float(A=$User.BubbleScale, B=30)
     ))
   ) {isSetParameters=true}
   &ScaleMeshSize(Scale Factor=&VectorFromFloat(Value=&Lerp_Float(A=0, B=1, Alpha=$StackContext.RegionMask.Float))) {enabled=false, module=ScaleMeshSize001}
   /SetVariables(Particles.Position=&AddVectorToPosition(
     Position=$Particles.Initial.Position
     Vector=&MakeVector(
      X=0
      Y=&Lerp_Float(
       A=1000
       B=-1200
       Alpha=&FloatFromCurve(FloatCurve=DataInterface() {propertyValues {Curve {
           defaultValue=1
           keys [
            {arriveTangent=4.1785, interpMode=RCIM_Cubic, leaveTangent=4.1785, tangentMode=RCTM_User, time=0, value=0}
            {arriveTangent=1.1179, interpMode=RCIM_Cubic, leaveTangent=1.1179, tangentMode=RCTM_User, time=0.2042, value=0.6228}
            {arriveTangent=0.0224, interpMode=RCIM_Cubic, leaveTangent=0.0224, tangentMode=RCTM_User, time=1, value=1}
           ]
          }}})
      )
      Z=&Lerp_Float(
       A=-750
       B=0
       Alpha=&FloatFromCurve(FloatCurve=DataInterface() {propertyValues {Curve {
           defaultValue=1
           keys [
            {interpMode=RCIM_Cubic, tangentMode=RCTM_User, time=0, value=0}
            {arriveTangent=3.9164, interpMode=RCIM_Cubic, leaveTangent=3.9164, tangentMode=RCTM_User, time=0.4, value=1}
           ]
          }}})
      )
     )
    )) {isSetParameters=true}
   &Color(Color=<rgba 0.1176 0.1753 0.25 1>) {enabled=false}
  ]
  rendererClasses [&NiagaraMeshRendererProperties, &NiagaraSpriteRendererProperties]
  renderers [
   {rendererClass=&NiagaraMeshRendererProperties, rendererIndex=0}
   {rendererClass=&NiagaraSpriteRendererProperties, rendererIndex=1}
   {rendererClass=&NiagaraSpriteRendererProperties, rendererIndex=2}
   {rendererClass=&NiagaraSpriteRendererProperties, rendererIndex=3}
   {rendererClass=&NiagaraSpriteRendererProperties, rendererIndex=4}
  ]
 }
 Foam {
  bEnabled=false
  emitterUpdateScript [
   &EmitterState()
   &SpawnBurst_Instantaneous(Use Spawn Probability=false, UseLoopCountLimit=false) {enabled=false}
   &SpawnRate(SpawnRate=&RandomRangeFloat(Minimum=0.5, Maximum=2), Use Spawn Probability=false)
  ]
  particleSpawnScript [
   &InitializeParticle(
    Color=<rgba 0.9498 0.9952 1 0.35>
    Lifetime Max=6
    Lifetime Min=5
    Lifetime Mode=::NewEnumerator1
    Sprite Size Mode=::NewEnumerator3
    Uniform Sprite Size=7500
    UsePositionOffset=true
    Write Material Random=true
   )
   &ShapeLocation(
    MaskSpawnGroup=false
    Offset Coordinate Space=::World
    Offset Mode=::NewEnumerator0
    Override Seed=false
    Ring Radius=1
    Rotation Coordinate Space=::World
    Shape Primitive=::NewEnumerator4
    UseOwnerScale=true
   )
   &AddVelocity(Module.Use Falloff Distance=false, Override Seed=false, Velocity Mode=::NewEnumerator1, Velocity Speed=376.7971)
   &ApplyInitialForces()
  ]
  particleUpdateScript [
   &ParticleState(Kill Particles When Lifetime Has Elapsed=false, Let Infinitely Lived Particles Die When Emitter Deactivates=true)
   /SetVariables(Particles.SpriteFacing=<0 0 1>) {isSetParameters=true}
   &DynamicMaterialParameters(
    Index 0 Param 2=1
    Index 0 Param 3=1
    Index 0 Param 4=&FloatFromCurve(FloatCurve=DataInterface() {propertyValues {Curve {keys [{tangentMode=RCTM_User, time=0, value=0}, {tangentMode=RCTM_User, time=1, value=1}]}}})
    Module.Param0WriteEnabled=true
    Module.Param1WriteEnabled=true
    Module.Param2WriteEnabled=true
    Module.Param3WriteEnabled=true
   )
   &Drag(Use Linear Drag=true, Use Rotational Drag=false)
   &SolveForcesAndVelocity(Clamp Velocity=false, Limit Acceleration=false)
  ]
  rendererClasses [&NiagaraSpriteRendererProperties]
  renderers [{rendererClass=&NiagaraSpriteRendererProperties}]
 }
}
summary {
 systemName=NS_FoulTame_WaterContact
 emitters [
  {bEnabled=false, emitterName=Foam, rendererClasses [&NiagaraSpriteRendererProperties]}
  {emitterName=Bubbles, rendererClasses [&NiagaraMeshRendererProperties, &NiagaraSpriteRendererProperties]}
 ]
 userVariables [User.BubbleScale {type=NiagaraFloat, defaultValue=45, description=""}]
}
