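# Niagara impact flare effect for FoulTame, made of two short-lived, large sprite bursts pushed toward the camera.
# Flare spawns at a socket chosen by ColorIndex modulo 5, then scales up and down while spinning; its VariantIndex/ColorIndex color-switch modules are disabled, leaving the initial purple.
# Ring bursts slightly later, expands and fades, follows a point on the user skeletal mesh, and computes extra flare color and size attributes.
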NiagaraSystem /WildEstate_FoulTame/Assets/FX/NS_FoulTame_ZP_ImpactFlare
dependencies {
 usedDataInterfaces [&NiagaraDataInterfaceCurve, &NiagaraDataInterfaceColorCurve]
 usedDynamicInputs []
 usedModules [&SystemState]
}
emitters {
 Flare {
  emitterUpdateScript [
   &EmitterState()
   &SpawnBurst_Instantaneous(Use Spawn Probability=false, UseLoopCountLimit=false)
  ]
  particleSpawnScript [
   &InitializeParticle(
    Color=<rgba 2.4893 0 5 1>
    Lifetime=&RandomRangeFloat(Minimum=0.3, Maximum=0.3)
    Sprite Rotation Angle=45
    Sprite Rotation Mode=::NewEnumerator0
    Sprite Size Mode=::NewEnumerator3
    Uniform Sprite Size=10000
    UsePositionOffset=true
    Write Material Random=true
   )
   &SocketLocation(
    Orientation Sampling=::NewEnumerator1
    SamplePosOffset=false
    Socket Index=&ModuloInt(A=$User.ColorIndex, B=5)
    Socket Reader=$System.Socket Reader
    Socket Sampling Mode=::NewEnumerator3
    VelocitySpeedLimit=false
   )
   &CameraOffset(Camera Offset Amount=10000)
  ]
  particleUpdateScript [
   &ParticleState()
   &ScaleSpriteSize(Uniform Curve Sprite Scale=DataInterface() {propertyValues {Curve {keys [
        {arriveTangent=0.9776, interpMode=RCIM_Cubic, leaveTangent=0.9776, tangentMode=RCTM_User, time=0, value=0}
        <at 0.3 1 RCIM_Cubic>
        {arriveTangent=-0.8026, interpMode=RCIM_Cubic, leaveTangent=-0.8026, tangentMode=RCTM_User, time=1, value=0}
       ]}}})
   &SpriteRotationRate(Rotation Rate=&FloatFromCurve(FloatCurve=DataInterface() {propertyValues {Curve {keys [<at 0.2 1 RCIM_Cubic>,<at 0.4 180 RCIM_Cubic>]}}}))
   &ScaleColor(
    Scale Alpha=&FloatFromCurve(FloatCurve=DataInterface() {propertyValues {Curve {keys [<at 0.6 1>,<at 1 2>]}}})
    Scale RGB=&MakeVectorFromLinearColorRGB(Color=&MergeHelper_ColorSwitch(
      0=&MergeHelper_ColorSwitch(
       0=&ColorFromCurve(ColorCurve=DataInterface() {propertyValues {
          AlphaCurve {keys [<at 0 1 RCIM_Cubic>,<at 1 1 RCIM_Cubic>]}
          BlueCurve {keys [<at 0.0283 5>,<at 0.4401 0 RCIM_Cubic>]}
          GreenCurve {keys [<at 0.0283 4.7557>,<at 0.4401 0 RCIM_Cubic>]}
          RedCurve {keys [<at 0.0283 3.1251>,<at 0.4401 0 RCIM_Cubic>]}
         }})
       1=&ColorFromCurve(ColorCurve=DataInterface() {propertyValues {
          AlphaCurve {keys [<at 0 1 RCIM_Cubic>,<at 1 1 RCIM_Cubic>]}
          BlueCurve {keys [<at 0.0394 0.9982>,<at 0.2509 0.2093>,<at 0.6029 0.0837>,<at 0.9852 0.0219>]}
          GreenCurve {keys [<at 0.0394 1.2586>,<at 0.2509 0>,<at 0.6029 0>,<at 0.9852 0>]}
          RedCurve {keys [<at 0.0394 5>,<at 0.2509 5>,<at 0.6029 0.5417>,<at 0.9852 0.151>]}
         }})
       2=&ColorFromCurve(ColorCurve=DataInterface() {propertyValues {
          AlphaCurve {keys [<at 0 1 RCIM_Cubic>,<at 1 1 RCIM_Cubic>]}
          BlueCurve {keys [<at 0.0283 5>,<at 0.4401 0 RCIM_Cubic>]}
          GreenCurve {keys [<at 0.0283 4.7557>,<at 0.4401 0 RCIM_Cubic>]}
          RedCurve {keys [<at 0.0283 3.1251>,<at 0.4401 0 RCIM_Cubic>]}
         }})
       3=?
       Index=&ModuloInt(A=$User.ColorIndex, B=2)
       Num=::NewEnumerator0
      )
      1=&MergeHelper_ColorSwitch(
       0=&ColorFromCurve(ColorCurve=DataInterface() {propertyValues {
          AlphaCurve {keys [<at 0 1 RCIM_Cubic>,<at 1 1 RCIM_Cubic>]}
          BlueCurve {keys [<at 0.0394 0.9982>,<at 0.2509 5>,<at 0.6029 0.5417>,<at 0.9852 0.151>]}
          GreenCurve {keys [<at 0.0394 1.2586>,<at 0.2509 4.6848>,<at 0.6029 0.2545>,<at 0.9852 0.025>]}
          RedCurve {keys [<at 0.0394 5>,<at 0.2509 0>,<at 0.6029 0.0591>,<at 0.9852 0.0552>]}
         }})
       1=&ColorFromCurve(ColorCurve=DataInterface() {propertyValues {
          AlphaCurve {keys [<at 0 1 RCIM_Cubic>,<at 1 1 RCIM_Cubic>]}
          BlueCurve {keys [<at 0.0394 0.9982>,<at 0.2509 5>,<at 0.6029 0.5417>,<at 0.9852 0.151>]}
          GreenCurve {keys [<at 0.0394 1.2586>,<at 0.2509 4.6848>,<at 0.6029 0.2545>,<at 0.9852 0.025>]}
          RedCurve {keys [<at 0.0394 5>,<at 0.2509 0>,<at 0.6029 0.0591>,<at 0.9852 0.0552>]}
         }})
       2=&ColorFromCurve(ColorCurve=DataInterface() {propertyValues {
          AlphaCurve {keys [<at 0 1 RCIM_Cubic>,<at 1 1 RCIM_Cubic>]}
          BlueCurve {keys [<at 0.0283 5>,<at 0.4401 0 RCIM_Cubic>]}
          GreenCurve {keys [<at 0.0283 4.7557>,<at 0.4401 0 RCIM_Cubic>]}
          RedCurve {keys [<at 0.0283 3.1251>,<at 0.4401 0 RCIM_Cubic>]}
         }})
       3=?
       Index=&ModuloInt(A=$User.ColorIndex, B=2)
       Num=::NewEnumerator0
      )
      2=?
      3=?
      Index=&MakeIntFromFloat(Float=$User.VariantIndex)
      Num=::NewEnumerator0
     ))
    ScaleA=true
    ScaleRGB=true
   ) {enabled=false}
   &ScaleColor(
    Scale Alpha=&FloatFromCurve(FloatCurve=DataInterface() {propertyValues {Curve {keys [<at 0.6 1>,<at 1 2>]}}})
    Scale RGB=&MakeVectorFromLinearColorRGB(Color=&MergeHelper_ColorSwitch(
      0=&MergeHelper_ColorSwitch(
       0=&ColorFromCurve(ColorCurve=DataInterface() {propertyValues {
          AlphaCurve {keys [<at 0 1 RCIM_Cubic>,<at 1 1 RCIM_Cubic>]}
          BlueCurve {keys [<at 0.0283 5>,<at 0.4401 0 RCIM_Cubic>]}
          GreenCurve {keys [<at 0.0283 4.7557>,<at 0.4401 0 RCIM_Cubic>]}
          RedCurve {keys [<at 0.0283 3.1251>,<at 0.4401 0 RCIM_Cubic>]}
         }})
       1=&ColorFromCurve(ColorCurve=DataInterface() {propertyValues {
          AlphaCurve {keys [<at 0 1 RCIM_Cubic>,<at 1 1 RCIM_Cubic>]}
          BlueCurve {keys [<at 0.0394 0.9982>,<at 0.2509 0.2093>,<at 0.6029 0.0837>,<at 0.9852 0.0219>]}
          GreenCurve {keys [<at 0.0394 1.2586>,<at 0.2509 0>,<at 0.6029 0>,<at 0.9852 0>]}
          RedCurve {keys [<at 0.0394 5>,<at 0.2509 5>,<at 0.6029 0.5417>,<at 0.9852 0.151>]}
         }})
       2=&ColorFromCurve(ColorCurve=DataInterface() {propertyValues {
          AlphaCurve {keys [<at 0 1 RCIM_Cubic>,<at 1 1 RCIM_Cubic>]}
          BlueCurve {keys [<at 0.0283 5>,<at 0.4401 0 RCIM_Cubic>]}
          GreenCurve {keys [<at 0.0283 4.7557>,<at 0.4401 0 RCIM_Cubic>]}
          RedCurve {keys [<at 0.0283 3.1251>,<at 0.4401 0 RCIM_Cubic>]}
         }})
       3=?
       Index=&ModuloInt(A=$User.ColorIndex, B=2)
       Num=::NewEnumerator0
      )
      1=&MergeHelper_ColorSwitch(
       0=&ColorFromCurve(ColorCurve=DataInterface() {propertyValues {
          AlphaCurve {keys [<at 0 1 RCIM_Cubic>,<at 1 1 RCIM_Cubic>]}
          BlueCurve {keys [<at 0.0283 5>,<at 0.9819 0 RCIM_Cubic>]}
          GreenCurve {keys [<at 0.0283 4.7557>,<at 0.9819 0 RCIM_Cubic>]}
          RedCurve {keys [<at 0.0283 3.1251>,<at 0.9819 0 RCIM_Cubic>]}
         }})
       1=&ColorFromCurve(ColorCurve=DataInterface() {propertyValues {
          AlphaCurve {keys [<at 0 1 RCIM_Cubic>,<at 1 1 RCIM_Cubic>]}
          BlueCurve {keys [<at 0.0394 0.9982>,<at 0.2509 5>,<at 0.6029 0.5417>,<at 0.9852 0.151>]}
          GreenCurve {keys [<at 0.0394 1.2586>,<at 0.2509 4.6848>,<at 0.6029 0.2545>,<at 0.9852 0.025>]}
          RedCurve {keys [<at 0.0394 5>,<at 0.2509 0>,<at 0.6029 0.0591>,<at 0.9852 0.0552>]}
         }})
       2=&ColorFromCurve(ColorCurve=DataInterface() {propertyValues {
          AlphaCurve {keys [<at 0 1 RCIM_Cubic>,<at 1 1 RCIM_Cubic>]}
          BlueCurve {keys [<at 0.0283 5>,<at 0.4401 0 RCIM_Cubic>]}
          GreenCurve {keys [<at 0.0283 4.7557>,<at 0.4401 0 RCIM_Cubic>]}
          RedCurve {keys [<at 0.0283 3.1251>,<at 0.4401 0 RCIM_Cubic>]}
         }})
       3=?
       Index=&ModuloInt(A=$User.ColorIndex, B=2)
       Num=::NewEnumerator0
      )
      2=?
      3=?
      Index=&MakeIntFromFloat(Float=$User.VariantIndex)
      Num=::NewEnumerator0
     ))
    ScaleA=true
    ScaleRGB=true
   ) {enabled=false, module=ScaleColor001}
  ]
  rendererClasses [&NiagaraSpriteRendererProperties]
  renderers [
   {rendererClass=&NiagaraSpriteRendererProperties}
   {rendererClass=&NiagaraSpriteRendererProperties, rendererIndex=1}
  ]
 }
 Ring {
  emitterSpawnScript [/SetVariables(Emitter.GlowDepthFade=25) {isSetParameters=true}]
  emitterUpdateScript [
   &EmitterState(Life Cycle Mode=::NewEnumerator1, Loop Behavior=::NewEnumerator1, Loop Delay=0.3875, Loop Duration=3.1125, UseLoopDelay=false)
   &SpawnBurst_Instantaneous(Spawn Count=1, Spawn Time=0.1, Use Spawn Probability=false)
  ]
  particleSpawnScript [
   &InitializeParticle(
    Color=<rgba 0.5098 0 1 1>
    Lifetime=0.2
    Position Offset Coordinate Space=::Local
    Sprite Rotation Mode=::NewEnumerator2
    Sprite Size Mode=::NewEnumerator3
    Uniform Sprite Size=5000
    UsePositionOffset=true
    Write Material Random=true
   )
   &CameraOffset(Camera Offset Amount=10000)
  ]
  particleUpdateScript [
   &ParticleState()
   &ScaleColor(
    Scale Alpha=&FloatFromCurve(FloatCurve=DataInterface() {propertyValues {Curve {keys [<at 0 0.5>,<at 1 0>]}}})
    Scale RGB=&MakeVectorFromLinearColorRGB(Color=&ColorFromCurve(ColorCurve=DataInterface() {propertyValues {
        AlphaCurve {keys [<at 0 1 RCIM_Cubic>,<at 1 0 RCIM_Cubic>]}
        BlueCurve {keys [<at 0.0117 1>,<at 0.0974 5 RCIM_Cubic>,<at 1 1 RCIM_Cubic>]}
        GreenCurve {keys [<at 0.0117 1>,<at 0.0974 5 RCIM_Cubic>,<at 1 1 RCIM_Cubic>]}
        RedCurve {keys [<at 0.0117 1>,<at 0.0974 5 RCIM_Cubic>,<at 1 1 RCIM_Cubic>]}
       }}))
    ScaleA=true
    ScaleRGB=true
   )
   &ScaleSpriteSize(Scale Factor=&Vector2DFromFloat(Value=&FloatFromCurve(FloatCurve=DataInterface() {propertyValues {Curve {
         defaultValue=0
         keys [
          {arriveTangent=4.1785, interpMode=RCIM_Cubic, leaveTangent=4.1785, tangentMode=RCTM_User, time=0, value=0}
          {arriveTangent=1.1179, interpMode=RCIM_Cubic, leaveTangent=1.1179, tangentMode=RCTM_User, time=0.2042, value=0.6228}
          {arriveTangent=0.0224, interpMode=RCIM_Cubic, leaveTangent=0.0224, tangentMode=RCTM_User, time=1, value=1}
         ]
        }}})))
   /SetVariables(
    Particles.FlareColor=&MultiplyLinearColor(Linear Color=$Particles.Initial.Color, Scale Factor=&MakeVector4_FromFloat(Float=7))
    Particles.FlareSize:Vector2f=&Multiply_Vector2D_ByFloat(
     Float=&FloatFromCurve(FloatCurve=DataInterface() {propertyValues {Curve {keys [
          {interpMode=RCIM_Cubic, tangentMode=RCTM_User, time=0, value=0}
          {interpMode=RCIM_Cubic, tangentMode=RCTM_User, time=0.3, value=0.95}
          {tangentMode=RCTM_User, time=1, value=0}
         ]}}})
     Vector2D=$Particles.Initial.SpriteSize
    )
    Particles.FlashCameraOffset=20
   ) {isSetParameters=true}
   &SkeletalMeshLocation(
    Bone / Socket Index=2
    Bone Sampling Mode=::NewEnumerator4
    MaskSpawnGroup=false
    Sampled Position Offset=<0 0 -15>
    Sampled Position Offset Coordinate Space:ENiagaraSkelMeshTransforms=::NewEnumerator2
    Skeletal Mesh=$User.SkeletalMesh
   )
  ]
  rendererClasses [&NiagaraSpriteRendererProperties]
  renderers [
   {rendererClass=&NiagaraSpriteRendererProperties}
   {rendererClass=&NiagaraSpriteRendererProperties, rendererIndex=1}
  ]
 }
}
summary {
 systemName=NS_FoulTame_ZP_ImpactFlare
 emitters [
  {emitterName=Flare, rendererClasses [&NiagaraSpriteRendererProperties]}
  {emitterName=Ring, rendererClasses [&NiagaraSpriteRendererProperties]}
 ]
 userVariables [
  User.Camera Offset Amount {type=NiagaraFloat, defaultValue=500, description=""}
  User.Color {type=LinearColor, description="", defaultValue {struct=LinearColor, value=<rgba 1 1 1 1>}}
  User.ColorIndex {type=NiagaraInt32, defaultValue=1}
  User.FadeDistance {type=NiagaraFloat, defaultValue=500000, description=""}
  User.Size {type=NiagaraFloat, defaultValue=2, description=""}
  User.SkeletalMesh {
   type=NiagaraDataInterfaceSkeletalMesh
   defaultValue {
    struct=NiagaraExt_VariableValue_DataInterface
    value {dataInterface=&self:NiagaraDataInterfaceSkeletalMesh_0}
   }
  }
  User.VariantIndex {type=NiagaraFloat, defaultValue=0, description=""}
 ]
}
