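# M_Gradient is an unlit translucent material that builds a soft opacity gradient, alongside a dark blue color parameter.
# An inverted Fresnel with a parameterized exponent makes the opacity value strongest where surfaces face the camera and fade toward grazing edges.
# That value feeds a DepthFade node with a parameterized distance, softening intersections with opaque geometry behind it.
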Material /WildEstate_FoulTame/Assets/Materials/M_Gradient
graph {
 properties {bAutomaticallySetUsageInEditor=false, blendMode=BLEND_Translucent, shadingModel=MSM_Unlit, usedShadingModels=Unlit}
 nodes [
  n1 = ScalarParameter() {defaultValue=3, parameterName=Fresnel}
  DepthFade(
   Opacity=OneMinus(None=Fresnel(ExponentIn=@n1) {baseReflectFraction=0})
   FadeDistance=ScalarParameter() {defaultValue=519.4956, parameterName=DepthFade}
  )
  VectorParameter() {defaultValue=<rgba 0.0692 0.148 0.25 1>}
 ]
}
